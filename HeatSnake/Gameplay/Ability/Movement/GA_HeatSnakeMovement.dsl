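; Gameplay ability for the HeatSnake grapple-knife zip, pulling the player to a target with root-motion MoveToForce.
; Targeting helpers handle ledges and predicted moving-platform motion, shaping offsets and exit velocity.
; The client detects capsule hits, computes wall (input-steered) or pawn bounces, then notifies the server, which can defer movement changes to post-movement updates.
; Zip end applies cooldown and fall-damage immunity; launchers and teleporters end the ability.
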(event Custom|DebugZipToTargetLocation
  (bind _targetlocation (Variables|Transient|GetTargetLocation))
  ;[1 0.4279 0] DEBUG: Display World/Relative ZipToTarget Locations
  (Development|PrintString (Utilities|String|Append "ZipToWorldLoc: " (Utilities|String|ToString(Vector) _targetlocation)) true true "(R=0.000000,G=0.660000,B=1.000000,A=1.000000)" 2.0 (Utilities|String|SelectName "ServerKnifeLoc" "ServerKnifeLoc" (Networking|IsDedicatedServer)))
  (Rendering|Debug|DrawDebugSphere _targetlocation 5e+01 12 (Math|Color|SelectColor "(R=1.000000,G=0.000000,B=0.017050,A=1.000000)" "(R=0.000000,G=1.000000,B=0.116588,A=1.000000)" (Networking|IsDedicatedServer)) 0.05))

(event Custom|OnServerPostMovementUpdated
  (params (DeltaSeconds :type Float%20%28single-precision%29) (OldLocation :type Vector) (OldVelocity :type Vector))
  ;[1 1 1] Activate from Event: EVENT.HEATSNAKE.SECONDARY.MOVEMENT
  ; If Hit location is less than X degrees use GrappleMovement otherwise active Swinging MME
  ;;[0.07619 0.2159 0.4072] Setup server post-movement update
  (MovementChangeDefer|UpdateServerMovementUpdatedFrame)
  (if (Variables|Transient|GetbServerMovementPending)
    (Variables|Transient|SetbServerMovementPending false)
    ;[0.07619 0.2159 0.4072] Movement ~
    (CallFunction|SetUseHighAnimationTolerance true)
    (MovementFunctions|OnPreMove)
    (Camera|PopCameraMode)
    (bind _asynctaskproxy (Ability|Tasks|ApplyRootMotionMoveToForce "GrappleKnifeMovement" (Variables|Transient|GetTargetLocation) (Variables|Movement|GetZipDuration) false "MOVE_Falling" false (Variables|Movement|GetPath%20Offset%20Curve) "SetVelocity" (MovementFunctions|AddAdditionalExitVelocity (* (TargetFunctions|CalculateDirtoTarget (Variables|Transient|GetTargetLocation)) (* (TargetFunctions|CalculateMovementSpeed (Variables|Transient|GetTargetLocation)) (MovementFunctions|GetExitSpeedMultiplier)))))
      (:then
        (Variables|AsyncTasks|SetMove%20Task _asynctaskproxy)
        (Ability|Tasks|WaitGameplayTagAdd "(TagName=\"Gameplay.Action.Player.Launched\")"
          (:then
            (Ability|Tasks|WaitGameplayTagAdd "(TagName=\"Abilities.Creative.Teleporter.JustTeleported\")"
              (:then
                (Ability|Tasks|WaitGameplayTagAdd "(TagName=\"Ability.DirectionalLaunch.Applied\")"
                  (:then
                    (Utilities|FlowControl|Sequence
                      (:then_0
                        ;[1 0.4279 0] DEBUG: Display World/Relative ZipToTarget Locations
                        (if (Variables|Movement|GetbDebugZipToTargetLocation)
                          (Utilities|Time|SetTimerbyEvent (delegate Custom|DebugZipToTargetLocation) 0.05 true)))
                      (:then_1
                        ;[1 1 1] Set Camera Mode
                        (Camera|PushCameraMode "/HeatSnake/Gameplay/Camera/HeatSnake_PlayerCameraMode_MME.HeatSnake_PlayerCameraMode_MME_C"))
                      (:then_2
                        ;[1 0.8541 0 1 12] Prevent falling damage
                        (GameplayCue|AddGameplayCueOnActor(Looping) (Variables|Transient|GetAbilityOwnerPawn) (Variables|Cues|GetGC_GrappleLoop) (Ability|GameplayCue|MakeGameplayCueParameters :NormalizedMagnitude 0.0 :RawMagnitude 0.0 :Location "0,0,0" :Normal (TargetFunctions|CalculateDirtoTarget (Variables|Transient|GetTargetLocation)) :GameplayEffectLevel 1 :AbilityLevel 1 :bReplicateLocationWhenUsingMinimalRepProxy false)))
                      (:then_3
                        ;[0.2075 1 0.3026 1 12] Client: Register for collision events
                        (Utilities|FlowControl|Delay (* (Variables|Movement|GetZipDuration) (ScalableFloat|GetValueatLevel (Variables|Hotfix|GetHF_BounceTimerMultiplier))))
                        (Ability|HasAuthority
                          (:False
                            (Collision|BindEventtoOnComponentHit (Variables|/Script/Engine.Character|GetCapsuleComponent (Variables|Transient|GetAbilityOwnerPawn)) (delegate Custom|On%20Component%20Hit)))))
                      (:then_4
                        _)))
                  (:Added
                    _)) ; Stop zip movement when launched from D-Launchers
                (Utilities|FlowControl|DoOnce true
                  (:Completed
                    (Ability|EndAbility))) ; Only call End once, in case we hit both a Bouncer and a Launcher.\nStart closed in case we already have the tag.
              )
              (:Added
                _)) ; Stop zip movement when teleported
          )
          (:Added
            _)) ; Stop zip movement when launched from bouncer devices
      )
      (:OnTimedOut
        _)
      (:OnTimedOutAndDestinationReached
        _))
    (MovementFunctions|OnPostMove true)
    (CallFunction|OnEndZipMovement false)
    (else
      (ServerHits|ProcessPendingServerHits))))

(event Custom|MMEMovementDelayedCoolDown
  ;[1 1 1] Activate from Event: EVENT.HEATSNAKE.SECONDARY.MOVEMENT
  ; If Hit location is less than X degrees use GrappleMovement otherwise active Swinging MME
  (Ability|CommitAbilityCooldownWithDuration (ScalableFloat|GetValueatLevel (Variables|Hotfix|GetHF_MMEMovementCooldownDuration))))

(event Ability|EventActivateAbilityFromEvent
  (params (EventData :type Gameplay%20Event%20Data%20Structure%20%28by%20ref%29)))

(event Custom|ServerActivate
  (params (ZipDuration :type Float%20%28double-precision%29) (PathOffset :type Curve%20Vector%20Object%20Reference) (InputDirection :type Integer) (Distance :type Float%20%28double-precision%29) (TargetLocation :type Vector) (TargetComp :type Scene%20Component%20Object%20Reference) (TargetCompRelLoc :type Vector) (bTargetWalkableSurface :type Boolean) (bTargetLedge :type Boolean))
  ;[0.07619 0.2159 0.4072] Store movement information on server
  (Variables|Movement|SetZipDuration ZipDuration)
  (Variables|Movement|SetPath%20Offset%20Curve PathOffset)
  (Variables|Movement|SetInputDirection InputDirection)
  (Variables|Transient|SetDistance Distance)
  (Variables|Transient|SetTargetLocation TargetLocation)
  (Variables|Transient|SetTargetComp TargetComp)
  (Variables|Transient|SetTargetCompRelLoc TargetCompRelLoc)
  (Variables|Transient|SetbTargetingWalkableSurface bTargetWalkableSurface)
  (Variables|Transient|SetbTargetingLedge bTargetLedge)
  (if (MovementChangeDefer|ShouldDeferServerMovementChanges)
    (Variables|Transient|SetbServerMovementPending true)
    (else
      ;[0.07619 0.2159 0.4072] Movement ~
      (CallFunction|SetUseHighAnimationTolerance true)
      (MovementFunctions|OnPreMove)
      (Camera|PopCameraMode)
      (bind _asynctaskproxy (Ability|Tasks|ApplyRootMotionMoveToForce "GrappleKnifeMovement" (Variables|Transient|GetTargetLocation) (Variables|Movement|GetZipDuration) false "MOVE_Falling" false (Variables|Movement|GetPath%20Offset%20Curve) "SetVelocity" (MovementFunctions|AddAdditionalExitVelocity (* (TargetFunctions|CalculateDirtoTarget (Variables|Transient|GetTargetLocation)) (* (TargetFunctions|CalculateMovementSpeed (Variables|Transient|GetTargetLocation)) (MovementFunctions|GetExitSpeedMultiplier)))))
        (:then
          (Variables|AsyncTasks|SetMove%20Task _asynctaskproxy)
          (Ability|Tasks|WaitGameplayTagAdd "(TagName=\"Gameplay.Action.Player.Launched\")"
            (:then
              (Ability|Tasks|WaitGameplayTagAdd "(TagName=\"Abilities.Creative.Teleporter.JustTeleported\")"
                (:then
                  (Ability|Tasks|WaitGameplayTagAdd "(TagName=\"Ability.DirectionalLaunch.Applied\")"
                    (:then
                      (Utilities|FlowControl|Sequence
                        (:then_0
                          ;[1 0.4279 0] DEBUG: Display World/Relative ZipToTarget Locations
                          (if (Variables|Movement|GetbDebugZipToTargetLocation)
                            (Utilities|Time|SetTimerbyEvent (delegate Custom|DebugZipToTargetLocation) 0.05 true)))
                        (:then_1
                          ;[1 1 1] Set Camera Mode
                          (Camera|PushCameraMode "/HeatSnake/Gameplay/Camera/HeatSnake_PlayerCameraMode_MME.HeatSnake_PlayerCameraMode_MME_C"))
                        (:then_2
                          ;[1 0.8541 0 1 12] Prevent falling damage
                          (GameplayCue|AddGameplayCueOnActor(Looping) (Variables|Transient|GetAbilityOwnerPawn) (Variables|Cues|GetGC_GrappleLoop) (Ability|GameplayCue|MakeGameplayCueParameters :NormalizedMagnitude 0.0 :RawMagnitude 0.0 :Location "0,0,0" :Normal (TargetFunctions|CalculateDirtoTarget (Variables|Transient|GetTargetLocation)) :GameplayEffectLevel 1 :AbilityLevel 1 :bReplicateLocationWhenUsingMinimalRepProxy false)))
                        (:then_3
                          ;[0.2075 1 0.3026 1 12] Client: Register for collision events
                          (Utilities|FlowControl|Delay (* (Variables|Movement|GetZipDuration) (ScalableFloat|GetValueatLevel (Variables|Hotfix|GetHF_BounceTimerMultiplier))))
                          (Ability|HasAuthority
                            (:False
                              (Collision|BindEventtoOnComponentHit (Variables|/Script/Engine.Character|GetCapsuleComponent (Variables|Transient|GetAbilityOwnerPawn)) (delegate Custom|On%20Component%20Hit)))))
                        (:then_4
                          _)))
                    (:Added
                      _)) ; Stop zip movement when launched from D-Launchers
                  (Utilities|FlowControl|DoOnce true
                    (:Completed
                      (Ability|EndAbility))) ; Only call End once, in case we hit both a Bouncer and a Launcher.\nStart closed in case we already have the tag.
                )
                (:Added
                  _)) ; Stop zip movement when teleported
            )
            (:Added
              _)) ; Stop zip movement when launched from bouncer devices
        )
        (:OnTimedOut
          _)
        (:OnTimedOutAndDestinationReached
          _))
      (MovementFunctions|OnPostMove true)
      (CallFunction|OnEndZipMovement false))))

(event Custom|On%20Component%20Hit
  (params (HitComponent :type Primitive%20Component%20Object%20Reference) (OtherActor :type Actor%20Object%20Reference) (OtherComp :type Primitive%20Component%20Object%20Reference) (NormalImpulse :type Vector) (Hit :type Hit%20Result%20Structure%20%28by%20ref%29))
  (bind (_bblockinghit _binitialoverlap _time _distance _location _impactpoint _normal) (Collision|BreakHitResult Hit))
  ;[0.2075 1 0.3026 1 12] Client: Register for collision events
  (CallFunction|LogComponentHit HitComponent OtherActor OtherComp Hit)
  (EventDispatchers|BindEventtoOnPawnLanded (Variables|Transient|GetAbilityOwnerPawn) (EventDispatchers|CreateEvent :!event Custom|ComponentHitOnLanded))
  (if _bblockinghit
    (bind _asfort_pawn (Utilities|Casting|CastToFortPawn OtherActor)
      (:then
        (CallFunction|ClientonHitPawn _location _normal _asfort_pawn))
      (:CastFailed
        (if (CallFunction|TreatHitNormalasWall _normal)
          (CallFunction|ClientonHitWall _location _normal))))
    (CallFunction|UnbindComponentHit)))

(event Custom|UnbindComponentHit
  ;[0.5208 1 0.537] Unbind component hit
  (Collision|UnbindEventfromOnComponentHit (Variables|/Script/Engine.Character|GetCapsuleComponent (Variables|Transient|GetAbilityOwnerPawn)) (EventDispatchers|CreateEvent :!event Custom|On%20Component%20Hit)))

(event Custom|EventOnAttackInAir
  ;[0.5576 1 0.5867] End ability and unbind bounce events after player attacks
  (bind _asynctaskproxy (Ability|Async|WaitGameplayEventToActor (Ability|GetActivatingPawn) "(TagName=\"Event.HeatSnake.Primary.Attack\")")
    (:then
      (Variables|AsyncTasks|SetWaitForPrimaryAttack _asynctaskproxy))
    (:EventReceived
      (CallFunction|UnbindComponentHit))))

(event Ability|EventOnEndAbility
  (params (bWasCancelled :type Boolean))
  ;[0.5576 1 0.5867] End Event
  (|CallParentFunction bWasCancelled)
  (CallFunction|SetUseHighAnimationTolerance false)
  (Development|PrintString "END GA_HeatSnakeMovement")
  (Character|UnbindEventfromOnCharacterMovementUpdated (Variables|Transient|GetAbilityOwnerPawn) (EventDispatchers|CreateEvent :!event Custom|OnServerPostMovementUpdated))
  (Camera|PopCameraMode))

(event Custom|OnEndZipMovement
  (params (InputDirection :type Boolean))
  ;[1 1 1] On Zip Movement End
  (Ability|HasAuthority
    (:True
      ;;[0 0.3157 1 1 12] Server No Collision: Wait for move task to finish and the client to notify the server that there was no collision. Client will end ability.
      (if (Variables|Transient|GetHadNoCollision)
        (Development|PrintString "Zip: No Collision")
        (CallFunction|CallWeapononEndZip false)))
    (:False
      ;;[0.3375 1 0.4114 1 12] Client No Collision: End the ability after a delay. (If a collision was detected, nothing more needs to be done here.)
      (if (not (Math|Boolean|ORBoolean (Variables|Transient|GetHadGroundCollision) (Variables|Transient|GetHadWallCollision) (Variables|Transient|GetHadPawnCollision)))
        (Variables|Transient|SetHadNoCollision true)
        (EventDispatchers|BindEventtoOnPawnLanded (Variables|Transient|GetAbilityOwnerPawn) (EventDispatchers|CreateEvent :!event Custom|OnLandedEvent))
        (CallFunction|NotifyServerNoCollision)
        (Development|PrintString "Zip: No Collision")
        (CallFunction|CallWeapononEndZip false)
        (Utilities|FlowControl|Delay 0.5)
        (if (not InputDirection)
          (CallFunction|UnbindComponentHit))
        (Ability|EndAbility)))))

(event Custom|NotifyServer_NoCollision
  ;[1 1 1] On Zip Movement End
  ;;[0 0.3157 1 1 12] Server No Collision: Wait for move task to finish and the client to notify the server that there was no collision. Client will end ability.
  (Variables|Transient|SetHadNoCollision true)
  (Utilities|IsValid (Variables|AsyncTasks|GetMove%20Task)
    (:Is%20Not%20Valid
      (Development|PrintString "Zip: No Collision")
      (CallFunction|CallWeapononEndZip false))))

(event Custom|CallWeaponOnEndZip
  (params (Bounced :type Boolean) (BounceVelocity :type Vector))
  ;[0.4211 1 0.8156] Call Weapon Blueprint OnEndZip
  ; Apply Delayed Cooldown for Zip Movement
  (GameplayCue|RemoveGameplayCue (Variables|Transient|GetAbilityOwnerPawn) (Variables|Cues|GetGC_GrappleLoop))
  (GameplayEffects|ApplyGameplayEffectToSelf :self (Ability|GetAbilitySystemComponent (Variables|Transient|GetAbilityOwnerPawn)) :GameplayEffectClass (Variables|Effects|GetGE_FallDamageImmunity_GAParent) :Level 0.0)
  (if Bounced
    (GameplayCue|ExecuteGameplayCueOnActor(Burst) (Variables|Transient|GetAbilityOwnerPawn) (Variables|Cues|GetGC_MovementBounceImpact) (Ability|GameplayCue|MakeGameplayCueParameters 0.0 (Math|Float|MapRangeClamped (TargetFunctions|CalculateMovementSpeed BounceVelocity) 0.0 (ScalableFloat|GetValueasDouble (Variables|Hotfix|GetHF_BaseZipSpeed)) 0.0 1.0))) ; Use velocity to drive magnitude of Burst
    (else
      _))
  (Ability|CommitAbilityCooldownWithDuration (ScalableFloat|GetValueatLevel (Variables|Hotfix|GetHF_GrappleMovementCooldownDuration)) true))

(event Custom|Check_Traveling_Tag
  ;[0.2038 0.3474 1] Check to verify grapple travling tag isn't added more than once
  ;;[1 0.52 0] Activate Movement Custom Verb
  (CustomVerb|BroadcastCustomVerbMessage (CustomVerb|MakeCustomVerbMessage (Ability|GetActivatingPawn) "(TagName=\"CustomVerb.HeatSnake.Movement.Activated\")")))

(event Custom|ClientOnHitWall
  (params (HitLocation :type Vector) (HitNormal :type Vector))
  ;[1 0.3157 0.8434] OnHitWall - Bounce off the wall taking into account  using direcitonal input.
  ;;[0.252 1 0.3643] Client
  (Development|PrintString "Zip Wall Collision")
  (Variables|Transient|SetHadWallCollision true)
  (CallFunction|EndMovementTask)
  (bind (_newbouncedirection _newbouncedistance _newbounceheight) (CallFunction|AdjustBounceforPlayerInput HitNormal (ScalableFloat|GetValueatLevel (Variables|Hotfix|GetHF_WallBounceDistance)) (ScalableFloat|GetValueatLevel (Variables|Hotfix|GetHF_WallBounceHeight))))
  (bind _clientbouncevelocity (CallFunction|ComputeBounceVelocity _newbouncedirection _newbouncedistance _newbounceheight))
  (CallFunction|ServerNotifyWallCollision (Transformation|GetActorLocation (Variables|Transient|GetAbilityOwnerPawn)) _clientbouncevelocity)
  (CallFunction|CallWeapononEndZip true _clientbouncevelocity)
  (CallFunction|Bounce _clientbouncevelocity)
  (Utilities|FlowControl|Delay 1.0)
  (Ability|EndAbility))

(event Custom|OnLandedEvent
  (params (Hit :type Hit%20Result%20Structure%20%28by%20ref%29))
  ;[0.2038 0.3474 1] On Landed remove bind event
  (CallFunction|UnbindComponentHit)
  (CallFunction|RemoveTravelingTag)
  (EventDispatchers|UnbindEventfromOnPawnLanded (Variables|Transient|GetAbilityOwnerPawn) (delegate Custom|OnLandedEvent))
  (TouchControls|RestoreTouchScreenRegions))

(event Custom|ServerNotify_WallCollision
  (params (ClientLocation :type Vector) (BounceVelocity :type Vector))
  ;[1 0.3157 0.8434] OnHitWall - Bounce off the wall taking into account  using direcitonal input.
  ;;[0.1864 0.5327 1] Server
  (Development|PrintString "Zip Wall Collision")
  (Variables|Transient|SetHadWallCollision true)
  (if (MovementChangeDefer|ShouldDeferServerMovementChanges)
    (Variables|Transient|SetbServerWallHitPending true)
    (Variables|Transient|SetServerPendingBounceVelocity BounceVelocity)
    (if (not (Variables|ConfigVariables|GetbDeferServerMovementChanges))
      (Utilities|FlowControl|DelayUntilNextTick) ; If bDeferServerMovementChanges \nis false, this needs to be here as it \ngets rid of all corrections.
    )
    (ServerHits|HandleServerWallHit BounceVelocity)
    (else
      _)))

(event Custom|Remove_TravelingTag
  ;[0.2038 0.3474 1] Remove traveling tag
  (Ability|GameplayEffect|RemoveLooseGameplayTags (Ability|GetActivatingPawn) (Variables|Default|GetQuest_TravelingTag) true))

(event Custom|ClientOnHitPawn
  (params (HitLocation :type Vector) (HitNormal :type Vector) (HitPawn :type Fort%20Pawn%20Object%20Reference))
  ;[0.624 0.2015 0.5475] OnHitPawn - small bounce to stay in range. At the moment, this is the same as the wall bounce except for some of the bounce values. We may need to do other stuff later.
  ;;[0.252 1 0.3643] Client
  (Development|PrintString "Zip Pawn Collision")
  (Variables|Transient|SetHadPawnCollision true)
  (CallFunction|EndMovementTask)
  (bind _clientbouncevelocity (CallFunction|ComputeBounceVelocity HitNormal (ScalableFloat|GetValueatLevel (Variables|Hotfix|GetHF_PawnBounceDistance)) (ScalableFloat|GetValueatLevel (Variables|Hotfix|GetHF_PawnBounceHeight))))
  (CallFunction|ServerNotifyPawnCollision (Transformation|GetActorLocation (Variables|Transient|GetAbilityOwnerPawn)) _clientbouncevelocity HitPawn)
  (CallFunction|CallWeapononEndZip false)
  (CallFunction|Bounce _clientbouncevelocity)
  (Utilities|FlowControl|Delay 0.5)
  (Ability|EndAbility))

(event Custom|ServerNotify_PawnCollision
  (params (ClientLocation :type Vector) (BounceVelocity :type Vector) (HitPawn :type Fort%20Pawn%20Object%20Reference))
  ;[0.624 0.2015 0.5475] OnHitPawn - small bounce to stay in range. At the moment, this is the same as the wall bounce except for some of the bounce values. We may need to do other stuff later.
  ;;[0.1864 0.5327 1] Server
  (Development|PrintString "Zip Pawn Collision")
  (GameplayCue|ExecuteGameplayCue HitPawn (Variables|Cues|GetGC_PawnHit))
  (Variables|Transient|SetHadPawnCollision true)
  (if (MovementChangeDefer|ShouldDeferServerMovementChanges)
    (Variables|Transient|SetbServerPawnHitPending true)
    (Variables|Transient|SetServerPendingBounceVelocity BounceVelocity)
    (else
      _))
  (if (not (Variables|ConfigVariables|GetbDeferServerMovementChanges))
    (Utilities|FlowControl|DelayUntilNextTick) ; If bDeferServerMovementChanges \nis false, this needs to be here as it \ngets rid of all corrections.
  )
  (ServerHits|HandleServerPawnHit BounceVelocity))

(event Custom|Bounce
  (params (BounceVelocity :type Vector))
  ;[1 0.3157 0.8434 1 12] Note that Apply Root Motion Jump Force prevents mantling, and does no better with net corrections.
  (Pawn|Character|LaunchCharacterJump (Variables|Transient|GetAbilityOwnerPawn) BounceVelocity true true true true)
  ;;[0.15 0.03047 0.03047 0.5] REMOVE GRAPPLE GC
  (GameplayCue|RemoveGameplayCue (Variables|Transient|GetAbilityOwnerPawn) (Variables|Cues|GetGC_GrappleLoop)))

(event Custom|ComponentHitOnLanded
  (params (Hit :type Hit%20Result%20Structure%20%28by%20ref%29))
  ;[1 1 1] Component hit and then on landed event to remove quest tag
  (CallFunction|RemoveTravelingTag)
  (EventDispatchers|UnbindEventfromOnPawnLanded (Variables|Transient|GetAbilityOwnerPawn) (delegate Custom|ComponentHitOnLanded))
  (TouchControls|RestoreTouchScreenRegions))

(fn AdaptTouchScreenRegions
  (bind _abilityownerpawn (Variables|Transient|GetAbilityOwnerPawn))
  (bind _returnvalue (Fortnite|GetFortPlayerControllerfromActor _abilityownerpawn))
  (bind _returnvalue_2 (Development|GetConsoleVariableBoolValue "HeatSnakeTouchControlsImprovement"))
  (if _returnvalue_2
    (if (Input|IsUsingTouch _returnvalue)
      (TouchControls|SetTouchRegionBounds :self _returnvalue :RotateCameraBounds (Math|Vector4|MakeVector4 0.0 1.0 0.0 1.0)))))

(fn AddAdditionalExitVelocity
  (params (BaseExitVelocity :type Vector))
  (if (MovingPlatformHelpers|IsTargetLastMovingPlatformInChain)
    ;[1 1 1] If we are moving to the last moving platform in a chain, give the player additional
    ; movement velicty along the same direction at the end to help them
    ; get on it.
    (return (+ BaseExitVelocity (* (* (Math|Float|Max(Float) (Math|Vector|DotProduct (Math|Vector|Normalize BaseExitVelocity) (Math|Vector|Normalize (MovingPlatformHelpers|GetInferredMovingPlatformVelocityinSeconds (Variables|Movement|GetZipDuration))))) 1.6) (MovingPlatformHelpers|GetInferredMovingPlatformVelocityinSeconds (Variables|Movement|GetZipDuration)))))
    (else
      (return BaseExitVelocity))))

(fn AdjustBounceForPlayerInput
  (params (CollisionNormal :type Vector) (BounceDistance :type Float%20%28double-precision%29) (BounceHeight :type Float%20%28double-precision%29))
  (bind _output_get (LocalVariables|SetPlayerInput (Math|Conversions|ToVector2D(Vector) (Pawn|Input|GetLastMovementInputVector (Variables|Transient|GetAbilityOwnerPawn)))))
  (LocalVariables|SetPlayerInputMagnitude (Math|Vector2D|Vector2DLength _output_get))
  (LocalVariables|SetPlayerInputDirection (Math|Vector2D|NormalSafe(Vector2D) (LocalVariables|GetPlayerInput)))
  ;[1 1 1] Has player movement input?
  (if (>= (LocalVariables|GetPlayerInputMagnitude) 0.25)
    ;[1 1 1] Is player input directed into the collision surface?
    (bind _output_get_2 (LocalVariables|SetPlayerInputAngleFromCollisionNormal (Math|Trig|Acos(Degrees) (Math|Vector2D|DotProduct(2D) (Math|Vector2D|Normalize2D (Math|Conversions|ToVector2D(Vector) (Variables|GetCollisionNormal))) (LocalVariables|GetPlayerInputDirection)))))
    (if (not (> _output_get_2 135.0))
      ;[1 1 1] Clamp player input direction a few degrees away from the collision normal, to try to make it looks more natural and to make it less likely to collide against the surface on the bounce back.
      (if (> (LocalVariables|GetPlayerInputAngleFromCollisionNormal) (ScalableFloat|GetValueatLevel (Variables|Hotfix|GetHF_WallBounceMaxAngle)))
        (LocalVariables|SetPlayerInputDirection (Math|Vector2D|GetRotated2D (LocalVariables|GetPlayerInputDirection) (- (ScalableFloat|GetValueatLevel (Variables|Hotfix|GetHF_WallBounceMaxAngle)) (LocalVariables|GetPlayerInputAngleFromCollisionNormal)))))
      ;[1 1 1] Bounce in direction of input, with added distance based on input magnitude.
      (return (Math|Conversions|ToVector(Vector2D) (Math|Vector2D|Normalize2D (LocalVariables|GetPlayerInput))) (+ (Variables|GetBounceDistance) (* (LocalVariables|GetPlayerInputMagnitude) (ScalableFloat|GetValueatLevel (Variables|Hotfix|GetHF_WallBounceDistance)))) (Variables|GetBounceHeight)))
    (else
      _))
  ;[1 1 1] Bounce up (and back just a bit) to get over a ledge
  (return (Variables|GetCollisionNormal) 1e+02 (+ (Variables|GetBounceHeight) (Variables|ConfigVariables|GetWallBounceHeightFromInput))))

(fn CalculateDirToTarget
  (params (Target :type Vector))
  (if (Utilities|IsValid (Variables|Transient|GetAbilityOwnerPawn) :!class /Script/Engine.KismetSystemLibrary)
    (return (Math|Vector|Normalize (- Target (Transformation|GetActorLocation (Variables|Transient|GetAbilityOwnerPawn)))))
    (else
      (return "0, 0, 0"))))

(fn CalculateMovementSpeed
  (params (Target :type Vector))
  (if (Utilities|IsValid (Variables|Transient|GetAbilityOwnerPawn) :!class /Script/Engine.KismetSystemLibrary)
    (return (Math|Float|SafeDivide (Math|Vector|Distance(Vector) Target (Transformation|GetActorLocation (Variables|Transient|GetAbilityOwnerPawn))) (Variables|Movement|GetZipDuration)))
    (else
      (return (ScalableFloat|GetValueatLevel (Variables|Hotfix|GetHF_BaseZipSpeed))))))

(fn ComputeBounceVelocity
  (params (BounceDirection :type Vector) (BounceDistance :type Float%20%28double-precision%29) (BounceHeight :type Float%20%28double-precision%29))
  (LocalVariables|SetGravity (Components|Movement|GetGravityZ (Variables|/Script/Engine.Character|GetCharacterMovement (Variables|Transient|GetAbilityOwnerPawn))))
  (LocalVariables|SetInitialVelocityZ (Math|Float|Sqrt (Utilities|Operators|Multiply (LocalVariables|GetGravity) -2.0 (Variables|GetBounceHeight))))
  (LocalVariables|SetJumpTime (* (/ (LocalVariables|GetInitialVelocityZ) (LocalVariables|GetGravity)) -1.0))
  (LocalVariables|SetHorizontalVelocity (/ (/ (Variables|GetBounceDistance) (LocalVariables|GetJumpTime)) 2.0))
  (return (+ (* (Math|Vector|Normalize2D(Vector) (Variables|GetBounceDirection)) (LocalVariables|GetHorizontalVelocity)) (Math|Vector|MakeVector 0.0 0.0 (LocalVariables|GetInitialVelocityZ)))))

(fn EndMovementTask
  (bind _move_task (Variables|AsyncTasks|GetMove%20Task))
  (Utilities|IsValid _move_task
    (:Is%20Valid
      (GameplayTasks|EndTask _move_task))
    (:Is%20Not%20Valid
      _))
  (Variables|AsyncTasks|SetMove%20Task 0)
  (MovementFunctions|OnPostMove false))

(fn GetExitSpeedMultiplier
  (return (Math|Float|SelectFloat (ScalableFloat|GetValueatLevel (Variables|Hotfix|GetHF_LedgeExitSpeedMult)) (Math|Float|SelectFloat (ScalableFloat|GetValueatLevel (Variables|Hotfix|GetHF_WalkableSurfaceExitSpeedMult)) (Math|Float|SelectFloat (ScalableFloat|GetValueatLevel (Variables|Hotfix|GetHF_HorizontalInputExitSpeedMult)) (ScalableFloat|GetValueatLevel (Variables|Hotfix|GetHF_StandardExitSpeedMult)) (!= (Variables|Movement|GetInputDirection) 0)) (Variables|Transient|GetbTargetingWalkableSurface)) (Variables|Transient|GetbTargetingLedge))))

(fn GetInferredMovingPlatformVelocityInSeconds
  (params (TimeDelta :type Float%20%28double-precision%29))
  (if (MovingPlatformHelpers|IsTargetMovingPlatform)
    (return (/ (- (.location (TargetFunctions|GetTargetTransforminSeconds TimeDelta)) (.location (TargetFunctions|GetTargetTransforminSeconds (- TimeDelta (LocalVariables|GetStepBackTimeDelta))))) (LocalVariables|GetStepBackTimeDelta)))
    (else
      (return "0, 0, 0"))))

(fn GetInputDirection
  (bind (_x _y) (Math|Vector2D|BreakVector2D (Math|Vector2D|NormalSafe(Vector2D) (Variables|/Script/FortniteGame.FortPlayerController|GetLastPlayerLocalMoveInput (Fortnite|GetFortPlayerControllerfromActor (Variables|Transient|GetAbilityOwnerPawn))) 1e-05)))
  (return (Math|Integer|SelectInt -1 (Math|Integer|SelectInt 1 0 (Math|Float|InRange(Float) _y 0.5)) (Math|Float|InRange(Float) _y -1.0 -0.5))))

(fn GetLedgeOffset
  (bind (_btargetingwalkablesurface _btargetingledge _ledgelocation) (CollisionChecks|IsClientImpactonLedge))
  (bind _abilityownerpawn (Variables|Transient|GetAbilityOwnerPawn))
  (bind _capsulecomponent (Variables|/Script/Engine.Character|GetCapsuleComponent _abilityownerpawn))
  (Variables|Transient|SetbTargetingWalkableSurface _btargetingwalkablesurface)
  (Variables|Transient|SetbTargetingLedge _btargetingledge)
  (if (and (Variables|Transient|GetbTargetingWalkableSurface) (not (Variables|Transient|GetbTargetingLedge)))
    (Variables|Transient|SetLedgeLocation (Variables|Transient|GetClientHitPoint))
    (else
      (Variables|Transient|SetLedgeLocation _ledgelocation)))
  (if (or (Variables|Transient|GetbTargetingWalkableSurface) (Variables|Transient|GetbTargetingLedge))
    (Utilities|IsValid _abilityownerpawn
      (:Is%20Valid
        (Utilities|IsValid _capsulecomponent
          (:Is%20Valid
            (return (Math|Vector|MakeVector 0.0 0.0 (+ (.z (- (Variables|Transient|GetLedgeLocation) (TargetFunctions|GetTargetWorldLocation 0.0))) (* (Variables|/Script/Engine.CapsuleComponent|GetCapsuleHalfHeight _capsulecomponent) (Math|Float|SelectFloat 2.0 1.5 (Variables|Transient|GetbTargetingLedge)))))))
          (:Is%20Not%20Valid
            _)))
      (:Is%20Not%20Valid
        _))
    (return "0, 0, 0")
    (else
      _)))

(fn GetMovementOffset
  (return (Math|Vector|MakeVector 0.0 (* (ScalableFloat|GetValueatLevel (Variables|Hotfix|GetHF_HorizontalInputOffset)) (Variables|Movement|GetInputDirection)))))

(fn GetTargetMovingPlatform
  (bind _targetcomp (Variables|Transient|GetTargetComp))
  (bind _returnvalue (Components|GetOwner _targetcomp))
  (bind _returnvalue_2 (Actor|GetComponentbyClass _returnvalue "/Script/FortniteGame.FortMovingPlatformComponent"))
  (Utilities|IsValid _targetcomp
    (:Is%20Valid
      (Utilities|IsValid _returnvalue_2
        (:Is%20Valid
          (return _returnvalue_2))
        (:Is%20Not%20Valid
          _)))
    (:Is%20Not%20Valid
      _))
  (return 0))

(fn GetTargetTransformInSeconds
  (params (TimeDelta :type Float%20%28double-precision%29))
  (bind _movingplatform (MovingPlatformHelpers|GetTargetMovingPlatform))
  (bind _targetcomp (Variables|Transient|GetTargetComp))
  ;[1 1 1] Handle the case if the target component is not valid
  (Utilities|IsValid (Variables|Transient|GetTargetComp)
    (:Is%20Valid
      ;[1 1 1] Handle if target is a Moving Platform
      (Utilities|IsValid _movingplatform
        (:Is%20Valid
          (return (Class|FortMovingPlatformComponent|GetPredictedTransformatServerTime _movingplatform (+ (Fortnite|Sync|Time|GetServerWorldTimeSeconds) (Variables|GetTimeDelta)))))
        (:Is%20Not%20Valid
          ;[1 1 1] Handle case of no time delta set
          (if (> (Variables|GetTimeDelta) 0)
            ;[1 1 1] Set Velocities
            (bind _asprimitive_component (Utilities|Casting|CastToPrimitiveComponent _targetcomp)
              (:then
                (bind _returnvalue (Physics|GetPhysicsLinearVelocity _asprimitive_component "None"))
                (LocalVariables|SetLinearVelocity (* _returnvalue (LocalVariables|GetVelocityMultiplier)))
                (LocalVariables|SetAngularVelocity (* (Physics|GetPhysicsAngularVelocityinDegrees _asprimitive_component "None") (LocalVariables|GetVelocityMultiplier))))
              (:CastFailed
                (LocalVariables|SetLinearVelocity (* (Transformation|GetComponentVelocity _targetcomp) (LocalVariables|GetVelocityMultiplier)))))
            (return (Math|Transform|MakeTransform (+ (.location (Transformation|GetWorldTransform (Variables|Transient|GetTargetComp))) (* (LocalVariables|GetLinearVelocity) (Variables|GetTimeDelta))) (Math|Rotator|MakeRotator (.x (+ (Math|Vector|MakeVector (.roll (.rotation (Transformation|GetWorldTransform (Variables|Transient|GetTargetComp)))) (.pitch (.rotation (Transformation|GetWorldTransform (Variables|Transient|GetTargetComp)))) (.yaw (.rotation (Transformation|GetWorldTransform (Variables|Transient|GetTargetComp))))) (* (LocalVariables|GetAngularVelocity) (Variables|GetTimeDelta)))) (.y (+ (Math|Vector|MakeVector (.roll (.rotation (Transformation|GetWorldTransform (Variables|Transient|GetTargetComp)))) (.pitch (.rotation (Transformation|GetWorldTransform (Variables|Transient|GetTargetComp)))) (.yaw (.rotation (Transformation|GetWorldTransform (Variables|Transient|GetTargetComp))))) (* (LocalVariables|GetAngularVelocity) (Variables|GetTimeDelta)))) (.z (+ (Math|Vector|MakeVector (.roll (.rotation (Transformation|GetWorldTransform (Variables|Transient|GetTargetComp)))) (.pitch (.rotation (Transformation|GetWorldTransform (Variables|Transient|GetTargetComp)))) (.yaw (.rotation (Transformation|GetWorldTransform (Variables|Transient|GetTargetComp))))) (* (LocalVariables|GetAngularVelocity) (Variables|GetTimeDelta))))) (.scale (Transformation|GetWorldTransform (Variables|Transient|GetTargetComp)))))
            (else
              (return (Transformation|GetWorldTransform (Variables|Transient|GetTargetComp))))))))
    (:Is%20Not%20Valid
      (return (Math|Transform|MakeTransform (Variables|Transient|GetClientHitPoint) (Math|Rotator|MakeRotfromX (Variables|Transient|GetClientHitNormal)))))))

(fn GetTargetWorldLocation
  (params (TimeDelta :type Float%20%28double-precision%29))
  (return (Math|Transform|TransformLocation (TargetFunctions|GetTargetTransforminSeconds TimeDelta) (Variables|Transient|GetTargetCompRelLoc))))

(fn HandleServerPawnHit
  (params (BounceVelocity :type Vector))
  (CallFunction|EndMovementTask)
  (CallFunction|CallWeapononEndZip false)
  (CallFunction|Bounce BounceVelocity))

(fn HandleServerWallHit
  (params (BounceVelocity :type Vector))
  (CallFunction|EndMovementTask)
  (CallFunction|CallWeapononEndZip true BounceVelocity)
  (CallFunction|Bounce BounceVelocity))

(fn IsTargetLastMovingPlatforminChain
  (bind _movingplatform (MovingPlatformHelpers|GetTargetMovingPlatform))
  (Utilities|IsValid _movingplatform
    (:Is%20Valid
      (return (Class|FortMovingPlatformComponent|IsFinalMovingPlatforminChain _movingplatform)))))

(fn IsClientHitInsideMovingPlatform
  (bind (_origin _boxextent) (Math|BoxSphereBounds|BreakBoxSphereBounds (Fortnite|Utility|GetLocalBounds (Variables|Transient|GetTargetComp))))
  (return (Math|Boolean|ANDBoolean (<= (.x (Math|Vector|VectorGetAbs (Variables|Transient|GetTargetCompRelLoc))) (.x (- _boxextent "5.000000,5.000000,5.000000"))) (<= (.y (Math|Vector|VectorGetAbs (Variables|Transient|GetTargetCompRelLoc))) (.y (- _boxextent "5.000000,5.000000,5.000000"))) (<= (.z (Math|Vector|VectorGetAbs (Variables|Transient|GetTargetCompRelLoc))) (.z (- _boxextent "5.000000,5.000000,5.000000"))))))

(fn IsClientImpactOnLedge
  (bind _targetcomp (Variables|Transient|GetTargetComp))
  (bind _returnvalue (Components|GetOwner _targetcomp))
  (bind (_origin _boxextent) (Collision|GetActorBounds _returnvalue true false))
  (Utilities|IsValid _targetcomp
    (:Is%20Valid
      (if (MovingPlatformHelpers|ShouldTreatBoundsasLedge _returnvalue)
        ;[1 1 1] Always target top of moving platform/car collision as if it's a ledge, as long as the player isn't based on it
        (return true (CollisionChecks|ShouldAutoTargetLedges) (Math|Vector|MakeVector (.x (Variables|Transient|GetClientHitPoint)) (.y (Variables|Transient|GetClientHitPoint)) (Math|Float|Max(Float) (.z (Variables|Transient|GetClientHitPoint)) (+ (.z _origin) (.z _boxextent)))))
        (else
          ;[1 1 1] Normal Ledge Detection
          (bind (_outdidimpactwalkablesurface _outtargetingledge _outledgelocation) (Class|GrappleWeaponLibrary|IsTargetingLedge (Variables|Transient|GetAbilityOwnerPawn) _returnvalue (Variables|Transient|GetClientHitPoint) (Variables|Transient|GetClientHitNormal) (Class|GrappleKnifeBPLib|GrappleKnifeGetLedgeConfig)))
          (return _outdidimpactwalkablesurface (select (CollisionChecks|ShouldAutoTargetLedges) _outtargetingledge false) _outledgelocation))))))

(fn IsLocationIndoors
  (params (Location :type Vector))
  (bind (_outhit _returnvalue) (Collision|LineTraceByProfile Location (+ Location (Math|Vector|MakeVector 0.0 0.0 (ScalableFloat|GetValueatLevel (Variables|Hotfix|GetHF_IndoorsUpwardsCheckDistance)))) "FortPawnCapsule"))
  (return _returnvalue))

(fn IsMovingIndoors
  (while (<= (LocalVariables|GetCurrentAlpha) 1.0)
    (bind _blocationindoors (CollisionChecks|IsLocationIndoors (Math|Vector|Lerp(Vector) (Transformation|GetActorLocation (Variables|Transient|GetAbilityOwnerPawn)) (TargetFunctions|GetTargetWorldLocation 0.0) (LocalVariables|GetCurrentAlpha))))
    (if _blocationindoors
      (return true)
      (else
        (LocalVariables|SetCurrentAlpha (+ (LocalVariables|GetCurrentAlpha) (LocalVariables|GetAlphaStep))))))
  (return false))

(fn IsOwnerBasedOnMovingPlatform
  (bind _abilityownerpawn (Variables|Transient|GetAbilityOwnerPawn))
  (bind _returnvalue (Pawn|GetMovementBaseActor _abilityownerpawn))
  (Utilities|IsValid _returnvalue
    (:Is%20Valid
      (return (Utilities|IsValid (Actor|GetComponentbyClass _returnvalue "/Script/FortniteGame.FortMovingPlatformComponent") :!class /Script/Engine.KismetSystemLibrary)))
    (:Is%20Not%20Valid
      (return false))))

(fn IsOwnerInsideMovingPlatform
  (bind (_origin _boxextent) (Collision|GetActorBounds (Pawn|GetMovementBaseActor (Variables|Transient|GetAbilityOwnerPawn)) true false))
  (if (MovingPlatformHelpers|IsOwnerBasedonMovingPlatform)
    (if (Math|Float|InRange(Float) (.z (Transformation|GetActorLocation (Variables|Transient|GetAbilityOwnerPawn))) (- (.z _origin) (.z _boxextent)) (+ (.z _origin) (.z _boxextent)))
      (return true)
      (else
        _)))
  (return false))

(fn IsTargetMovingPlatform
  (return (Utilities|IsValid (MovingPlatformHelpers|GetTargetMovingPlatform) :!class /Script/Engine.KismetSystemLibrary)))

(fn LogComponentHit
  (params (Hit%20Component :type Primitive%20Component%20Object%20Reference) (OtherActor :type Actor%20Object%20Reference) (OtherComponent :type Primitive%20Component%20Object%20Reference) (HitResult :type Hit%20Result%20Structure))
  (bind (_bblockinghit _binitialoverlap _time _distance _location) (Collision|BreakHitResult HitResult))
  (Development|PrintString (Utilities|String|BuildString(Vector) (Utilities|String|BuildString(Object) "Zip Hit " "Actor " OtherActor) "at" _location)))

(fn OnPostMove
  (params (bFinished :type Boolean))
  (if (Variables|Transient|GetbIsMoveActive)
    (Variables|Transient|SetbIsMoveActive false)))

(fn OnPreMove
  (if (not (Variables|Transient|GetbIsMoveActive))
    (Variables|Transient|SetbIsMoveActive true)
    (bind _asfort_player_pawn (Utilities|Casting|CastToFortPlayerPawn (Variables|Transient|GetAbilityOwnerPawn)))
    (Class|FortPlayerPawn|SetCrouch _asfort_player_pawn false false)
    (TouchControls|AdaptTouchScreenRegions)))

(fn ProcessPendingServerHits
  (if (Variables|Transient|GetbServerWallHitPending)
    (Variables|Transient|SetbServerWallHitPending false)
    (ServerHits|HandleServerWallHit (Variables|Transient|GetServerPendingBounceVelocity)))
  (if (Variables|Transient|GetbServerPawnHitPending)
    (Variables|Transient|SetbServerPawnHitPending false)
    (ServerHits|HandleServerPawnHit (Variables|Transient|GetServerPendingBounceVelocity))))

(fn ResetTransientVariables
  (Variables|Transient|SetHadGroundCollision false)
  (Variables|Transient|SetHadWallCollision false)
  (Variables|Transient|SetHadPawnCollision false)
  (Variables|Transient|SetHadNoCollision false))

(fn RestoreTouchScreenRegions
  (bind _abilityownerpawn (Variables|Transient|GetAbilityOwnerPawn))
  (bind _returnvalue (Fortnite|GetFortPlayerControllerfromActor _abilityownerpawn))
  (bind _returnvalue_2 (Development|GetConsoleVariableBoolValue "HeatSnakeTouchControlsImprovement"))
  (if _returnvalue_2
    (if (Input|IsUsingTouch _returnvalue)
      (TouchControls|SetTouchRegionBounds :self _returnvalue :MovePlayerBounds (Math|Vector4|MakeVector4 0.0 0.5 0.0 1.0) :RotateCameraBounds (Math|Vector4|MakeVector4 0.5 1.0 0.0 1.0)))))

(fn RoundVectorToInt
  (params (In :type Vector%20%28by%20ref%29))
  (return (Math|Vector|MakeVector (Math|Conversions|ToFloat(Integer) (Math|Float|Round (.x In))) (Math|Conversions|ToFloat(Integer) (Math|Float|Round (.y In))) (Math|Conversions|ToFloat(Integer) (Math|Float|Round (.z In))))))

(fn SelectCurveOffset
  (bind _bmovingindoors (CollisionChecks|IsMovingIndoors))
  (bind _output_get (Variables|Movement|SetPath%20Offset%20Curve (select (or _bmovingindoors (<= (Variables|Movement|GetZipDuration) (ScalableFloat|GetValueatLevel (Variables|Hotfix|GetHF_UseIndoorCurveZipTime)))) "/HeatSnake/Gameplay/Ability/Movement/Curves/HeatSnake_CurveOffset_InDoors_Movement.HeatSnake_CurveOffset_InDoors_Movement" (select (and (> (.z (TargetFunctions|CalculateDirtoTarget (Variables|Transient|GetTargetLocation))) 0) (Variables|Transient|GetbTargetingLedge)) "/HeatSnake/Gameplay/Ability/Movement/Curves/HeatSnake_CurveOffset_Ledge_Movement.HeatSnake_CurveOffset_Ledge_Movement" "/HeatSnake/Gameplay/Ability/Movement/Curves/HeatSnake_CurveOffset_Movement.HeatSnake_CurveOffset_Movement"))))
  (return _output_get))

(fn SetReticleState
  (params (NewReticleState :type GrappleKnifeReticleState%20Enum :default "NewEnumerator0") (ZipDistance :type Float%20%28double-precision%29) (NewWorldLocation :type Vector) (bTargetingLedge :type Boolean)))

(fn SetUseHighAnimationTolerance
  (params (Use%20HAT :type Boolean))
  (bind _asfort_movement_comp_character (Utilities|Casting|CastToFortMovementComp_Character (Variables|/Script/Engine.Character|GetCharacterMovement (Ability|GetActivatingPawn))))
  (NetTolerance|SetUseHighAccelerationTolerance _asfort_movement_comp_character Use%20HAT "ZipToMovement"))

(fn ShouldAutoTargetLedges
  (return (ScalableFloat|GetValueasBool (Variables|Hotfix|GetHF_AutoTargetLedges))))

(fn ShouldDeferServerMovementChanges
  (return (and (Variables|ConfigVariables|GetbDeferServerMovementChanges) (!= (Utilities|GetFrameCount) (Variables|Transient|GetServerMovementUpdatedFrame)))))

(fn ShouldTreatBoundsAsLedge
  (params (Actor :type Actor%20Object%20Reference))
  (Utilities|IsValid Actor
    (:Is%20Valid
      (return (Math|Boolean|ANDBoolean (CollisionChecks|ShouldAutoTargetLedges) (or (Utilities|IsValid (Actor|GetComponentbyClass Actor "/Script/FortniteGame.FortMovingPlatformComponent") :!class /Script/Engine.KismetSystemLibrary) (Utilities|IsValid (Actor|GetComponentbyClass Actor "/Script/ValetRuntime.DagwoodNetworkPhysicsComponent") :!class /Script/Engine.KismetSystemLibrary)) (not (or (MovingPlatformHelpers|IsOwnerInsideMovingPlatform) (and (MovingPlatformHelpers|IsOwnerBasedonMovingPlatform) (MovingPlatformHelpers|IsClientHitInsideMovingPlatform)))))))
    (:Is%20Not%20Valid
      (return false))))

(fn TreatHitNormalAsWall
  (params (A :type Vector))
  (return (< (Math|Vector|DotProduct A "0.000000,0.000000,1.000000") (Math|Trig|Cos(Degrees) (ScalableFloat|GetValueatLevel (Variables|Hotfix|GetHF_AngleWallVsGround))))))

(fn UpdateServerMovementUpdatedFrame
  (Variables|Transient|SetServerMovementUpdatedFrame (Utilities|GetFrameCount)))

(fn UpdateTargetComponent
  (params (Hit :type Hit%20Result%20Structure%20%28by%20ref%29))
  (bind (_bblockinghit _binitialoverlap _time _distance _location _impactpoint _normal _impactnormal _physmat _hitactor _hitcomponent) (Collision|BreakHitResult Hit))
  (bind (_origin _boxextent) (Math|BoxSphereBounds|BreakBoxSphereBounds (Fortnite|Utility|GetLocalBounds (Variables|Transient|GetTargetComp))))
  (Ability|HasAuthority
    (:False
      ;[1 1 1] Set Target comp and relative location
      (Variables|Transient|SetClientHitNormal _normal)
      (Variables|Transient|SetClientHitPoint _location)
      (bind _output_get (Variables|Transient|SetTargetComp _hitcomponent))
      (Utilities|IsValid _output_get
        (:Is%20Valid
          (Variables|Transient|SetTargetCompRelLoc (Math|Transform|InverseTransformLocation (Transformation|GetWorldTransform _hitcomponent) _impactpoint))
          ;[1 1 1] If should be treated as ledge, force relative location to be pulled inward on XY and set to the Z max.
          (if (MovingPlatformHelpers|ShouldTreatBoundsasLedge (Components|GetOwner _hitcomponent))
            (Variables|Transient|SetTargetCompRelLoc (Math|Vector|MakeVector (Math|Float|Clamp(Float) (.x (Variables|Transient|GetTargetCompRelLoc)) (* (Math|Float|Max(Float) (- (.x _boxextent) (Variables|ConfigVariables|GetBoundsAsLedgeRelLocPullInDistanceX))) -1.0) (Math|Float|Max(Float) (- (.x _boxextent) (Variables|ConfigVariables|GetBoundsAsLedgeRelLocPullInDistanceX)))) (Math|Float|Clamp(Float) (.y (Variables|Transient|GetTargetCompRelLoc)) (* (Math|Float|Max(Float) (- (.y _boxextent) (Variables|ConfigVariables|GetBoundsAsLedgeRelLocPullInDistanceY))) -1.0) (Math|Float|Max(Float) (- (.y _boxextent) (Variables|ConfigVariables|GetBoundsAsLedgeRelLocPullInDistanceY)))) (Math|Float|Max(Float) (.z (Variables|Transient|GetTargetCompRelLoc)) (.z _boxextent))))))
        (:Is%20Not%20Valid
          (Variables|Transient|SetTargetCompRelLoc "0, 0, 0"))))))

(fn UpdateTargetLocation
  (bind _ledgeoffset (TargetFunctions|GetLedgeOffset))
  (bind _movementoffset (TargetFunctions|GetMovementOffset))
  (bind _output_get (Variables|Transient|SetBaseTargetLocation (TargetFunctions|GetTargetWorldLocation (Variables|Movement|GetZipDuration))))
  (LocalVariables|SetDirToTarget (TargetFunctions|CalculateDirtoTarget _output_get))
  (Variables|Transient|SetTargetLocation (CallFunction|RoundVectortoInt (+ (Variables|Transient|GetBaseTargetLocation) (Class|GrappleWeaponLibrary|CalculateDirectionalTargetOffset (LocalVariables|GetDirToTarget) _movementoffset _ledgeoffset)))))
